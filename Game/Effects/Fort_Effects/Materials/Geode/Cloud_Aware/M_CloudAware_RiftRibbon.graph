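# Translucent unlit Niagara ribbon material for a Geode rift effect: glowing electric cracks blended with turbulent cloud, with cloud fogging enabled.
# Counter-panning turbulence samples break up the cloud layer, while a scrolling crack texture adds glow, both colored by MF_Geode_Crack_Colors.
# Particle color and dynamic parameters drive tint, crack spread, panning and HDR intensity.
# Depth, view-angle and camera-distance fades soften it; camera fade also scales an object-relative XY offset.
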Material /Game/Effects/Fort_Effects/Materials/Geode/Cloud_Aware/M_CloudAware_RiftRibbon
graph {
 properties {
  bApplyCloudFogging=true
  bUsedWithNiagaraRibbons=true
  blendMode=BLEND_Translucent
  pixelDepthOffsetMode=PDOM_Legacy
  shadingModel=MSM_Unlit
  usedShadingModels=Unlit
 }
 nodes [
  n1 = StaticBool() {value=true}
  n2 = NamedRerouteUsage()
  n3 = MaterialFunctionCall(XY=TextureCoordinate()) {materialFunction=&Swizzle}
  n4 = NamedRerouteUsage()
  n5 = ComponentMask(None=VectorParameter() {parameterName=Cloud UVs, defaultValue=<rgba 0.1 25 0 0>}.RGB) {g=true, r=true}
  n6 = Multiply(A=Panner(Coordinate=@n3.YX, Time=@n4) {speedX=1, speedY=0.005}, B=@n5)
  n7 = Multiply(A=Panner(Coordinate=@n3.YX, Time=@n4) {speedX=-1, speedY=0.005}, B=@n5)
  n8 = Multiply(A=LinearInterpolate(
    A=TextureSample(UVs=@n6) {texture=/Game/Global/Textures/Default/Caustics/T_Turbulence_Seamless}.R
    B=TextureSample(UVs=@n7) {texture=&T_Turbulence_Seamless}.R
    Alpha=NamedRerouteUsage()
   )) {constB=2}
  n9 = Multiply(
   A=Multiply(
    A=MaterialFunctionCall!(Use Alternate UVs=@n1, Alternate UVs=@n2) {materialFunction=/Game/Athena/Effects/Materials/Geode/Functions/MF_Geode_Crack_Colors}.Cloud Color
    B=NamedRerouteDeclaration(None=@n8) {name=NoiseBreakUp, nodeColor=<rgba 0 1 0.3203 1>}
   )
   B=ScalarParameter() {defaultValue=3, parameterName=Cloud Intensity}
  )
  n10 = TextureCoordinate()
  n11 = Power(Base=OneMinus(None=ComponentMask(None=@n10) {r=true}), Exp=ScalarParameter()) {constExponent=0.5}
  n12 = Multiply(A=Power(Base=NamedRerouteUsage()) {constExponent=5}, B=ScalarParameter() {defaultValue=0.2, parameterName=CloudGlowAdd}) {constB=0.15}
  n13 = Constant()
  n14 = VectorParameter() {parameterName=CrackGlowMult, defaultValue=<rgba 0.3893 0.3473 0.5052 1>}
  n15 = MaterialFunctionCall(Use Alternate UVs=StaticBool() {value=true}, Alternate UVs=NamedRerouteUsage()) {materialFunction=&MF_Geode_Crack_Colors}
  n16 = AppendVector(A=Constant() {r=1}, B=ScalarParameter() {defaultValue=1, parameterName=TileMult})
  n17 = TextureCoordinate() {vTiling=0.25}
  n18 = Add(
   A=Multiply(A=@n16, B=MaterialFunctionCall(XY=@n17) {materialFunction=&Swizzle}.YX)
   B=AppendVector(A=Constant(), B=NamedRerouteUsage())
  )
  n19 = TextureSample(UVs=@n18) {texture=/Game/Effects/Fort_Effects/Textures/Electric/T_Vert_Electric_Cracks}
  n20 = Add(A=@n19.R, B=@n19.R)
  n21 = ScalarParameter() {defaultValue=10, parameterName=Crack Intensity}
  n22 = Add(
   A=Multiply(
    A=MaterialFunctionCall(X=@n12, Z=@n13) {materialFunction=&MakeFloat3}.Result
    B=Multiply(A=@n14.RGB, B=@n15.Crack Color)
   )
   B=Multiply(A=@n15.Crack Color, B=Multiply(A=@n20, B=@n21))
  )
  n23 = LinearInterpolate(A=Constant3Vector() {constant=<rgba 0.2133 0.1829 0.2344 0>}, Alpha=NamedRerouteUsage())
  n24 = Multiply(A=ParticleColor().RGB, B=Add(A=LinearInterpolate(B=@n9, Alpha=@n11), B=Multiply(A=@n22, B=@n23)))
  n25 = Saturate(None=@n24)
  n26 = DynamicParameter() {parameterIndex=1, defaultValue=<rgba 1 1 1 0>, paramNames [HDR Intensity, Param2, Param3, Param4]}
  MaterialFunctionCall(Emissive=ShadingPathSwitch(
    Default=MaterialFunctionCall(LDR Input=@n25, HDR Tint=@n24, HDR Intensity Multiplier=@n26.HDR Intensity) {materialFunction=&DeriveHDRfromLDR}.RemappedRange
    Mobile=@n24
   )) {materialFunction=&MF_FX_EmissiveAdjustments}
  n27 = WorldPosition()
  n28 = ObjectPositionWS()
  n29 = MaterialFunctionCall(Distance=ScalarParameter() {defaultValue=65000, parameterName=Fade Distance}) {materialFunction=&Near_Camera_Fade}
  n30 = OneMinus(None=@n29.Result)
  Multiply(
   A=Multiply(A=Subtract(A=@n27.XYZ, B=@n28), B=Constant3Vector() {constant=<rgba 1 1 0 0>})
   B=LinearInterpolate(Alpha=Saturate(None=@n30)) {constA=1, constB=-1}
  )
  n31 = Power(Base=Sine(None=MaterialFunctionCall() {materialFunction=&LinearGradient}.VGradient) {period=2})
  n32 = NamedRerouteUsage()
  n33 = ScalarParameter() {defaultValue=8, parameterName=Fade Sensitivity}
  Multiply(
   A=Multiply(
    A=Multiply(A=@n31, B=@n32)
    B=MaterialFunctionCall("FadeSensitivity (optional)"=@n33) {materialFunction=&FadeBasedOnViewAngle}.Result
   )
   B=@n29.Result
  )
  n34 = DynamicParameter() {defaultValue=<rgba 1 1 1 0>, paramNames [Crack, SplineWidthMult, Pan, Open]}
  n35 = NamedRerouteDeclaration(None=LinearInterpolate(
    B=DynamicParameter() {defaultValue=<rgba 1 1 1 0>, paramNames [Crack, SplineWidthMult, Pan, Open]}.Crack
    Alpha=ScalarParameter() {defaultValue=1, parameterName=CrackSize}
   )) {name=Crack, nodeColor=<rgba 1 0.2578 0 1>}
  NamedRerouteDeclaration(None=StaticSwitchParameter(True=Multiply(A=@n34.Pan, B=@n35), False=@n34.Pan) {parameterName=Crack spread affects panning}) {name=Panning, nodeColor=<rgba 1 0.8672 0 1>}
  n36 = ComponentMask(None=TextureCoordinate()) {r=true}
  n37 = Clamp(None=Power(Base=Sine(None=@n36) {period=2}) {constExponent=0.6}) {minDefault=0.1}
  n38 = MaterialFunctionCall(
   UVs=MaterialFunctionCall(XY=TextureCoordinate()) {materialFunction=&Swizzle}.YX
   Texture Scale=AppendVector(A=@n37, B=Constant() {r=1})
  ) {materialFunction=&ScaleUVsByCenter}
  NamedRerouteDeclaration(None=SphereMask(A=ComponentMask(None=@n38.UVs) {r=true}, B=Constant() {r=0.5}) {attenuationRadius=0.5, hardnessPercent=0}) {name=SphereMask_UV0, nodeColor=<rgba 0 0.25 1 1>}
  NamedRerouteDeclaration(None=@n37) {name=Edge Mask, nodeColor=<rgba 0 1 0.2969 1>}
  NamedRerouteDeclaration(None=@n19.RGB) {name=MaskTexture Panner 2, nodeColor=<rgba 1 0 0.3984 1>}
  NamedRerouteDeclaration(None=TextureSample(UVs=@n18) {texture=&T_Vert_Electric_Cracks}.B) {name="Mask Texture Mask (B)", nodeColor=<rgba 0 0.25 1 1>}
  NamedRerouteDeclaration() {name=Normal Influence, nodeColor=<rgba 1 0 0.1406 1>}
  n39 = Multiply(A=Multiply(A=NamedRerouteUsage(), B=NamedRerouteUsage()) {constA=1}, B=Constant2Vector() {r=-0.06}) {constB=0.2}
  n40 = MaterialFunctionCall(XY=TextureCoordinate() {vTiling=1.5}) {materialFunction=&Swizzle}
  NamedRerouteDeclaration(None=Add(A=@n39, B=Multiply(A=@n40.YX, B=@n16))) {name=Panning Coord, nodeColor=<rgba 1 0.375 0 1>}
  NamedRerouteDeclaration(None=Multiply(A=Time(), B=ScalarParameter() {defaultValue=0.1, parameterName=PanSpeed})) {name=Panning Speed, nodeColor=<rgba 1 0 0.2578 1>}
  n41 = ScalarParameter() {defaultValue=2500, parameterName=DepthFadeDIst}
  n42 = NamedRerouteUsage()
  n43 = Constant3Vector() {constant=<rgba 1 1 0 0>}
  n44 = NamedRerouteUsage()
  n45 = ScalarParameter() {defaultValue=0.75, parameterName=Noise BreakUp Addition}
  n46 = Multiply(A=ParticleColor().A, B=Add(A=Multiply(A=@n42, B=@n43), B=Add(A=@n44, B=@n45) {constB=0.15}))
  NamedRerouteDeclaration(None=ShadingPathSwitch(Default=DepthFade(Opacity=@n46, FadeDistance=@n41) {fadeDistanceDefault=500}, Mobile=@n46)) {name=Opacity, nodeColor=<rgba 1 0.0469 0 1>}
  n47 = TextureCoordinate()
  n48 = ScalarParameter() {defaultValue=1, parameterName=Corruption}
  n49 = Power(Base=Sine(None=ComponentMask(None=@n47) {g=true}) {period=2}, Exp=LinearInterpolate(Alpha=@n48) {constA=10, constB=4})
  n50 = Multiply(A=@n48) {constB=5}
  NamedRerouteDeclaration(None=Clamp(None=Multiply(A=@n49, B=@n50) {constB=2})) {name=AltUVs, nodeColor=<rgba 1 0.7969 0 1>}
  NamedRerouteDeclaration(None=Add(A=@n39, B=@n40.YX)) {name=Cloud Panning Coorx, nodeColor=<rgba 0 1 0.8359 1>}
 ]
}
