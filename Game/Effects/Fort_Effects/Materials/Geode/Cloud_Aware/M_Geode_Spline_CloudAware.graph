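# Unlit translucent Geode effect material for mesh particles, drawing glowing electric cracks and colored cloud along a spline, with cloud fogging.
# Panning turbulence breaks up a rainbow gradient and Geode colors, layered with distorted crack textures.
# Dynamic parameters drive crack spread, panning, opening and HDR intensity, with SplineThicken and optional vibrating crack offset shaping the mesh.
# Sphere masks and depth fade shape opacity; quality and mobile switches trim cost.
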Material /Game/Effects/Fort_Effects/Materials/Geode/Cloud_Aware/M_Geode_Spline_CloudAware
graph {
 properties {
  bApplyCloudFogging=true
  bUsedWithMeshParticles=true
  bUsedWithNiagaraMeshParticles=true
  blendMode=BLEND_Translucent
  pixelDepthOffsetMode=PDOM_Legacy
  shadingModel=MSM_Unlit
  twoSided=true
  usedShadingModels=Unlit
 }
 nodes [
  n1 = TextureCoordinate() {coordinateIndex=1}
  n2 = Panner(Coordinate=NamedRerouteUsage(), Time=NamedRerouteUsage()) {speedX=0.1, speedY=0.05}
  n3 = Panner(Coordinate=NamedRerouteUsage(), Time=NamedRerouteUsage()) {speedX=-0.1, speedY=0.05}
  n4 = LinearInterpolate(
   A=TextureSample(UVs=@n2) {texture=/Game/Global/Textures/Default/Caustics/T_Turbulence_Seamless}.R
   B=TextureSample(UVs=@n3) {texture=&T_Turbulence_Seamless}.R
   Alpha=NamedRerouteUsage()
  )
  n5 = TextureSample(UVs=Add(A=Rotator(Coordinate=@n1) {speed=0.5}, B=Multiply(A=@n4) {constB=0.6})) {texture=/Game/Global/Textures/Default/Masks_and_Gradients/T_Grad_Rainbow}
  n6 = Add(
   A=QualitySwitch(Default=Multiply(A=@n5.RGB) {constB=0.2}, Low=Constant())
   B=MaterialFunctionCall!(Use Alternate UVs, Alternate UVs) {materialFunction=/Game/Athena/Effects/Materials/Geode/Functions/MF_Geode_Crack_Colors}.Cloud Color
  )
  n7 = NamedRerouteDeclaration(None=Multiply(A=@n4) {constB=2}) {name=NoiseBreakUp, nodeColor=<rgba 0 1 0.3203 1>}
  n8 = NamedRerouteUsage()
  n9 = Multiply(A=NamedRerouteUsage(), B=ScalarParameter() {defaultValue=1.5, parameterName=CloudGlowMult}) {constB=2}
  n10 = QualitySwitch(Default=Power(Base=Multiply(A=@n8, B=@n9)) {constExponent=3}, Low=Constant())
  n11 = Multiply(A=Power(Base=NamedRerouteUsage()) {constExponent=5}, B=ScalarParameter() {defaultValue=0.2, parameterName=CloudGlowAdd}) {constB=0.15}
  n12 = VectorParameter() {parameterName=CrackGlowMult, defaultValue=<rgba 0.3893 0.3473 0.5052 1>}
  n13 = MaterialFunctionCall() {materialFunction=&MF_Geode_Crack_Colors}
  n14 = QualitySwitch(Default=Multiply(A=NamedRerouteUsage(), B=Constant2Vector() {r=0.011}), Low=Constant())
  n15 = NamedRerouteDeclaration(None=LinearInterpolate(Alpha=NamedRerouteUsage()) {constA=1, constB=-1}) {name="Remaped Mask (B)", nodeColor=<rgba 1 0.6094 0 1>}
  n16 = AppendVector(A=Constant() {r=1}, B=ScalarParameter() {defaultValue=1, parameterName=TileMult})
  n17 = Add(A=Multiply(A=@n16, B=TextureCoordinate() {coordinateIndex=1}), B=AppendVector(A=Constant(), B=NamedRerouteUsage()))
  n18 = TextureSample(UVs=Add(A=Multiply(A=@n14, B=@n15), B=@n17)) {texture=/Game/Effects/Fort_Effects/Textures/Electric/T_Vert_Electric_Cracks}
  n19 = Add(A=@n18.R, B=@n18.R)
  n20 = LinearInterpolate(Alpha=NamedRerouteUsage()) {constA=0.8, constB=2}
  n21 = Add(A=Multiply(A=Add(A=@n10, B=@n11), B=Multiply(A=@n12.RGB, B=@n13.Crack Color)), B=Multiply(A=@n13.Crack Color, B=Power(Base=@n19, Exp=@n20)))
  n22 = LinearInterpolate(A=Constant3Vector() {constant=<rgba 0.2133 0.1829 0.2344 0>}, Alpha=NamedRerouteUsage())
  n23 = Multiply(A=ParticleColor().RGB, B=Add(A=Multiply(A=@n6, B=@n7), B=Multiply(A=@n21, B=@n22)))
  n24 = Saturate(None=@n23)
  n25 = DynamicParameter() {parameterIndex=1, defaultValue=<rgba 1 1 1 0>, paramNames [HDR Intensity, Param2, Param3, Param4]}
  MaterialFunctionCall(Emissive=ShadingPathSwitch(
    Default=MaterialFunctionCall(LDR Input=@n24, HDR Tint=@n23, HDR Intensity Multiplier=@n25.HDR Intensity) {materialFunction=&DeriveHDRfromLDR}.RemappedRange
    Mobile=@n23
   )) {materialFunction=&MF_FX_EmissiveAdjustments}
  n26 = ScalarParameter() {defaultValue=4000, parameterName=NearCameraWidthDisrance}
  n27 = Multiply(
   A=Multiply(
    A=DynamicParameter() {defaultValue=<rgba 1 1 1 0>, paramNames [Crack, SplineWidthMult, Pan, Open]}.SplineWidthMult
    B=ScalarParameter() {defaultValue=500, parameterName=Width}
   ) {constA=1}
   B=ShadingPathSwitch(
    Default=MaterialFunctionCall(Distance=@n26) {materialFunction=&Near_Camera_Fade}.Result
    Mobile=ScalarParameter() {defaultValue=1, parameterName=NearCameraWidth_Mobile}
   )
  )
  n28 = TextureCoordinate()
  n29 = DynamicParameter() {defaultValue=<rgba 1 1 1 0>, paramNames [Crack, SplineWidthMult, Pan, Open]}
  n30 = Multiply(
   A=@n29.Open
   B=MaterialFunctionCall!(Time=@n29.Open) {materialFunction=/Game/Effects/Fort_Effects/Materials/Geode/_Functions/MF_SineVibration}.Result
  )
  n31 = MaterialFunctionCall!(Mask=Power(Base=ComponentMask(None=@n28) {g=true}), Parameter=Add(A=@n29.Open, B=Multiply(A=@n30) {constB=0.05})) {materialFunction=/Game/Effects/Fort_Effects/Materials/Geode/_Functions/MF_Geode_Crack_WPO}
  n32 = Constant()
  NamedRerouteDeclaration(None=Add(
    A=MaterialFunctionCall(WidthBase=@n27, WidthTip=@n27) {materialFunction=&SplineThicken}.WorldPositionOffset
    B=StaticSwitchParameter(True=@n31.Result, False=@n32) {defaultValue=true, parameterName=UseWPO}
   ) {constB=0}) {name=WPO, nodeColor=<rgba 0.125 1 0 1>}
  n33 = DynamicParameter() {defaultValue=<rgba 1 1 1 0>, paramNames [Crack, SplineWidthMult, Pan, Open]}
  n34 = NamedRerouteDeclaration(None=LinearInterpolate(
    B=DynamicParameter() {defaultValue=<rgba 1 1 1 0>, paramNames [Crack, SplineWidthMult, Pan, Open]}.Crack
    Alpha=ScalarParameter() {defaultValue=1, parameterName=CrackSize}
   )) {name=Crack, nodeColor=<rgba 1 0.2578 0 1>}
  NamedRerouteDeclaration(None=StaticSwitchParameter(True=Multiply(A=@n33.Pan, B=@n34), False=@n33.Pan) {parameterName=Crack spread affects panning}) {name=Panning, nodeColor=<rgba 1 0.8672 0 1>}
  n35 = LinearInterpolate(
   A=TextureSample(UVs=@n2) {samplerType=SAMPLERTYPE_Normal, texture=/Game/Effects/Fort_Effects/Textures/Plasma/T_Turbulence_Seamless_N}.RGB
   B=TextureSample(UVs=@n3) {samplerType=SAMPLERTYPE_Normal, texture=&T_Turbulence_Seamless_N}.RGB
   Alpha=NamedRerouteUsage()
  )
  n36 = AppendVector(A=NamedRerouteUsage(), B=Constant2Vector())
  NamedRerouteDeclaration(None=Clamp(None=DotProduct(A=@n35, B=@n36))) {name=Normal Influence, nodeColor=<rgba 1 0 0.1406 1>}
  n37 = Multiply(A=NamedRerouteUsage(), B=NamedRerouteUsage()) {constA=1}
  n38 = Constant2Vector() {r=-0.06}
  n39 = TextureCoordinate() {coordinateIndex=1, uTiling=1.5}
  NamedRerouteDeclaration(None=Add(A=Multiply(A=@n37, B=@n38) {constB=0.2}, B=Multiply(A=@n39, B=@n16))) {name=Panning Coord, nodeColor=<rgba 1 0.375 0 1>}
  NamedRerouteDeclaration(None=Multiply(A=Time(), B=ScalarParameter() {defaultValue=0.1, parameterName=PanSpeed})) {name=Panning Speed, nodeColor=<rgba 1 0 0.2578 1>}
  n40 = ScalarParameter() {defaultValue=800, parameterName=DepthFadeDIst}
  n41 = VertexColor()
  n42 = NamedRerouteUsage()
  n43 = NamedRerouteUsage()
  n44 = Add(
   A=Multiply(A=NamedRerouteUsage(), B=Constant3Vector() {constant=<rgba 1 1 0 0>})
   B=Multiply(A=Add(A=@n42) {constB=0.15}, B=Power(Base=@n43)) {constA=1}
  )
  n45 = Multiply(A=ParticleColor().A, B=Multiply(A=Multiply(A=@n41.G, B=@n44), B=NamedRerouteUsage()))
  NamedRerouteDeclaration(None=ShadingPathSwitch(Default=DepthFade(Opacity=@n45, FadeDistance=@n40) {fadeDistanceDefault=500}, Mobile=@n45)) {name=Opacity, nodeColor=<rgba 0.1953 0 1 1>}
  n46 = ComponentMask(None=TextureCoordinate()) {g=true}
  n47 = TextureCoordinate() {coordinateIndex=2}
  n48 = Constant() {r=0.5}
  n49 = NamedRerouteUsage()
  n50 = LinearInterpolate(Alpha=@n49) {constB=2}
  n51 = LinearInterpolate(Alpha=@n49) {constA=0.7}
  n52 = Multiply(
   A=Clamp(None=Power(Base=@n46) {constExponent=0.6}) {minDefault=0.1}
   B=NamedRerouteDeclaration(None=SphereMask(A=@n47, B=@n48, Radius=@n50, Hardness=@n51) {hardnessPercent=0}) {name=SphereMask_UV2, nodeColor=<rgba 0 1 0.6719 1>}
  )
  n53 = MaterialFunctionCall(UVs=TextureCoordinate(), Texture Scale=AppendVector(A=@n52, B=Constant() {r=1})) {materialFunction=&ScaleUVsByCenter}
  NamedRerouteDeclaration(None=SphereMask(A=ComponentMask(None=@n53.UVs) {r=true}, B=Constant() {r=0.5}) {attenuationRadius=0.5, hardnessPercent=0}) {name=SphereMask_UV0, nodeColor=<rgba 0 0.25 1 1>}
  NamedRerouteDeclaration(None=@n52) {name=Edge Mask, nodeColor=<rgba 0 1 0.2969 1>}
  NamedRerouteDeclaration(None=@n18.RGB) {name=MaskTexture Panner 2, nodeColor=<rgba 1 0 0.3984 1>}
  NamedRerouteDeclaration(None=TextureSample(UVs=@n17) {texture=&T_Vert_Electric_Cracks}.B) {name="Mask Texture Mask (B)", nodeColor=<rgba 0 0.25 1 1>}
  NamedRerouteUsage()
  NamedRerouteUsage()
 ]
}
